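# Physics asset for the dangling cuff chains on the medium male Quarrel character body, defining their secondary-motion bodies and joints.
# Each of the two chain bones has an offset capsule constrained to its own small master-bone capsule, with uneven swing limits and very little twist.
# A single right-thigh capsule is the only other body, apparently a collider that keeps the chains out of the leg.
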PhysicsAsset /BRCosmetics/Characters/Player/Male/Medium/Bodies/M_MED_Quarrel/Meshes/M_MED_Quarrel_dyn_cuffs
bodies {
 dyn_cuffChain_01 [{shapeType=Capsule, length=2, radius=1.75, shapeName=dyn_cuffChain_01_capsule, center=<5.3 3 -1>, rotation=<rot -5.1328 -38.5866 -63.106>}]
 dyn_cuffChain_01_MSTR [{shapeType=Capsule, length=0.505, radius=0.505, shapeName=dyn_cuffChain_01_MSTR_capsule, rotation=<rot 0 -90 0>}]
 dyn_cuffChain_02 [{shapeType=Capsule, length=2, radius=1.75, shapeName=dyn_cuffChain_02_capsule, center=<6 -2.3 0>, rotation=<rot 3.6678 -132.5409 -69.8542>}]
 dyn_cuffChain_02_MSTR [{shapeType=Capsule, length=0.505, radius=0.505, shapeName=dyn_cuffChain_02_MSTR_capsule, rotation=<rot 0 -90 0>}]
 thigh_r [{shapeType=Capsule, length=14.5958, radius=8.1967, shapeName=thigh_r_capsule, center=<7.1228 -0.9272 0.799>, rotation=<rot 88.993 180 180>}]
}
constraints [
 {bone1Name=dyn_cuffChain_01, bone2Name=dyn_cuffChain_01_MSTR, swing1LimitDegrees=15, swing2LimitDegrees=45, twistLimitDegrees=2}
 {bone1Name=dyn_cuffChain_02, bone2Name=dyn_cuffChain_02_MSTR, swing1LimitDegrees=15, swing2LimitDegrees=45, twistLimitDegrees=2}
]
