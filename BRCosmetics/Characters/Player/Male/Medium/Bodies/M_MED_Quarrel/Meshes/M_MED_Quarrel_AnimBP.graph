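# Post-style AnimBlueprint for the M_MED_Quarrel cosmetic body that adds corrective deformation and secondary motion on the pose copied from its parent mesh.
# Two Control Rigs (twists, sculpters) feed constraint, look-at and additive bone nodes that adjust muscle volumes, belt, pouches and cuffs.
# AnimDynamics and RigidBody nodes simulate shoelaces, necklace, dog tags and cuffs.
# EventGraph derives driver alphas from limb deltas against reference pose, plus a pistol aim-up correction.
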AnimBlueprint /BRCosmetics/Characters/Player/Male/Medium/Bodies/M_MED_Quarrel/Meshes/M_MED_Quarrel_AnimBP
target_skeleton=&/M_MED_Quarrel_Skeleton
graphs [
 AnimGraph {nodes [
   n1 = Anim Dynamics - Bone: dyn_cuffChain_01(Alpha=1, ExternalForce="0,0,0") {
    angularDampingOverride=0.99
    angularSpringConstant=1500
    bAngularSpring=true
    bGravityOverrideInSimSpace=true
    bOverrideAngularDamping=true
    gravityScale=0.35
    simulationSpace=RootRelative
    alphaBoolBlend {blendOption=Linear}
    alphaScaleBiasClamp {scale=0.6}
    boundBone {boneName=dyn_cuffChain_01}
    componentAppliedLinearAccClamp {z=5000}
    componentLinearAccScale=<0.45 0.45 0.45>
    componentLinearVelScale=<0.45 0.45 0.45>
    physicsBodyDefinitions [{
      sphereCollisionRadius=0
      boundBone {boneName=dyn_cuffChain_01}
      boxExtents=<10 8 8>
      constraintSetup {angularLimitsMax=<0 0 15>, angularLimitsMin=<0 -45 -15>, angularTarget=<0 0 0>}
      localJointOffset=<10 0 0>
     }]
    relativeSpaceBone {boneName=pelvis}
   }
   Anim Dynamics - Bone: dyn_cuffChain_02(ComponentPose=@n1, Alpha=1, ExternalForce="0,0,0") {
    angularDampingOverride=0.99
    angularSpringConstant=1500
    bAngularSpring=true
    bGravityOverrideInSimSpace=true
    bOverrideAngularDamping=true
    gravityScale=0.35
    simulationSpace=RootRelative
    alphaBoolBlend {blendOption=Linear}
    alphaScaleBiasClamp {scale=0.6}
    boundBone {boneName=dyn_cuffChain_02}
    componentAppliedLinearAccClamp {z=5000}
    componentLinearAccScale=<0.45 0.45 0.45>
    componentLinearVelScale=<0.45 0.45 0.45>
    physicsBodyDefinitions [{
      sphereCollisionRadius=0
      boundBone {boneName=dyn_cuffChain_02}
      boxExtents=<10 8 8>
      constraintSetup {angularLimitsMax=<5 0 15>, angularLimitsMin=<-5 -45 -15>, angularTarget=<0 0 0>}
      localJointOffset=<10 0 0>
     }]
    relativeSpaceBone {boneName=pelvis}
   }
   n2 = "Constraint - dyn_volume_base_lowerarm_l (PP)"!(ComponentPose, Alpha=1, ConstraintWeights_0=0.65, ConstraintWeights_1:float=0.35) -> (Pose) {
    alphaBoolBlend {blendOption=Linear}
    boneToModify {boneName=dyn_volume_base_lowerarm_l}
    constraintSetup [{targetBone {boneName=upperarm_l}}, {targetBone {boneName=lowerarm_l}}]
    constraintWeights [1,1]
   }
   n3 = "Constraint - dyn_volume_base_lowerarm_r (PP)"!(ComponentPose=@n2, Alpha=1, ConstraintWeights_0=0.65, ConstraintWeights_1:float=0.35) -> (Pose) {
    alphaBoolBlend {blendOption=Linear}
    boneToModify {boneName=dyn_volume_base_lowerarm_r}
    constraintSetup [{targetBone {boneName=upperarm_r}}, {targetBone {boneName=lowerarm_r}}]
    constraintWeights [1,1]
   }
   n4 = "Transform (Modify) Bone - Bone: dyn_volume_lowerarm_l"(ComponentPose=@n3, Alpha=Get elbow_rot_l() -> (elbow_rot_l:double), Translation="0,-3,0", Rotation="0,0,0", Scale="0.3,1,1") {
    rotationMode=BMM_Additive
    rotationSpace=BCS_BoneSpace
    scaleMode=BMM_Additive
    scaleSpace=BCS_BoneSpace
    translationMode=BMM_Additive
    translationSpace=BCS_BoneSpace
    alphaBoolBlend {blendOption=Linear}
    alphaScaleBiasClamp {bMapRange=true, inRange {max=80}}
    boneToModify {boneName=dyn_volume_lowerarm_l}
    scale {x=0.3}
    translation {y=-3}
   }
   n5 = "Transform (Modify) Bone - Bone: dyn_volume_lowerarm_r"(ComponentPose=@n4, Alpha=Get elbow_rot_r() -> (elbow_rot_r:double), Translation="0,3,0", Rotation="0,0,0", Scale="0.3,1,1") {
    rotationMode=BMM_Additive
    rotationSpace=BCS_BoneSpace
    scaleMode=BMM_Additive
    scaleSpace=BCS_BoneSpace
    translationMode=BMM_Additive
    translationSpace=BCS_BoneSpace
    alphaBoolBlend {blendOption=Linear}
    alphaScaleBiasClamp {bMapRange=true, inRange {max=80}}
    boneToModify {boneName=dyn_volume_lowerarm_r}
    scale {x=0.3}
    translation {y=3}
   }
   n6 = "Constraint - dyn_volume_base_thigh_l (PP)"!(ComponentPose=@n5, Alpha=1, ConstraintWeights_0=0.6, ConstraintWeights_1:float=0.4) -> (Pose) {
    alphaBoolBlend {blendOption=Linear}
    boneToModify {boneName=dyn_volume_base_thigh_l}
    constraintSetup [{targetBone {boneName=pelvis}}, {targetBone {boneName=thigh_l}}]
    constraintWeights [1,1]
   }
   n7 = "Constraint - dyn_volume_base_thigh_r (PP)"!(ComponentPose=@n6, Alpha=1, ConstraintWeights_0=0.6, ConstraintWeights_1:float=0.4) -> (Pose) {
    alphaBoolBlend {blendOption=Linear}
    boneToModify {boneName=dyn_volume_base_thigh_r}
    constraintSetup [{targetBone {boneName=pelvis}}, {targetBone {boneName=thigh_r}}]
    constraintWeights [1,1]
   }
   n8 = Get thigh_rot_l()
   n9 = "Transform (Modify) Bone - Bone: dyn_volume_thigh_l"(ComponentPose=@n7, Alpha=@n8, Translation="-10,-5,0", Rotation="0,0,0", Scale="1,1,1") {
    rotationMode=BMM_Additive
    rotationSpace=BCS_BoneSpace
    scaleMode=BMM_Additive
    scaleSpace=BCS_BoneSpace
    translationMode=BMM_Additive
    translationSpace=BCS_BoneSpace
    alphaBoolBlend {blendOption=Linear}
    alphaScaleBiasClamp {bMapRange=true, inRange {max=0, min=-100}, outRange {max=0, min=1}}
    boneToModify {boneName=dyn_volume_thigh_l}
    translation=<-10 -5>
   }
   n10 = "Transform (Modify) Bone - Bone: dyn_volume_thigh_l"(ComponentPose=@n9, Alpha=@n8, Translation="0,-4,0", Rotation="0,0,0", Scale="1,1,1") {
    rotationMode=BMM_Additive
    rotationSpace=BCS_BoneSpace
    scaleMode=BMM_Additive
    scaleSpace=BCS_BoneSpace
    translationMode=BMM_Additive
    translationSpace=BCS_BoneSpace
    alphaBoolBlend {blendOption=Linear}
    alphaScaleBiasClamp {bMapRange=true, inRange {max=50}}
    boneToModify {boneName=dyn_volume_thigh_l}
    translation {y=-4}
   }
   n11 = Get thigh_rot_r()
   n12 = "Transform (Modify) Bone - Bone: dyn_volume_thigh_r"(ComponentPose=@n10, Alpha=@n11, Translation="10,2,0", Rotation="0,0,0", Scale="1,1,1") {
    rotationMode=BMM_Additive
    rotationSpace=BCS_BoneSpace
    scaleMode=BMM_Additive
    scaleSpace=BCS_BoneSpace
    translationMode=BMM_Additive
    translationSpace=BCS_BoneSpace
    alphaBoolBlend {blendOption=Linear}
    alphaScaleBiasClamp {bMapRange=true, inRange {max=0, min=-100}, outRange {max=0, min=1}}
    boneToModify {boneName=dyn_volume_thigh_r}
    translation=<10 2>
   }
   n13 = "Transform (Modify) Bone - Bone: dyn_volume_thigh_r"(ComponentPose=@n12, Alpha=@n11, Translation="0,4,0", Rotation="0,0,0", Scale="1,1,1") {
    rotationMode=BMM_Additive
    rotationSpace=BCS_BoneSpace
    scaleMode=BMM_Additive
    scaleSpace=BCS_BoneSpace
    translationMode=BMM_Additive
    translationSpace=BCS_BoneSpace
    alphaBoolBlend {blendOption=Linear}
    alphaScaleBiasClamp {bMapRange=true, inRange {max=50}}
    boneToModify {boneName=dyn_volume_thigh_r}
    translation {y=4}
   }
   n14 = "Constraint - dyn_volume_base_calf_l (PP)"!(ComponentPose=@n13, Alpha=1, ConstraintWeights_0=0.5, ConstraintWeights_1:float=0.5) -> (Pose) {
    alphaBoolBlend {blendOption=Linear}
    boneToModify {boneName=dyn_volume_base_calf_l}
    constraintSetup [{targetBone {boneName=thigh_l}}, {targetBone {boneName=calf_l}}]
    constraintWeights [1,1]
   }
   n15 = "Constraint - dyn_volume_base_calf_r (PP)"!(ComponentPose=@n14, Alpha=1, ConstraintWeights_0=0.5, ConstraintWeights_1:float=0.5) -> (Pose) {
    alphaBoolBlend {blendOption=Linear}
    boneToModify {boneName=dyn_volume_base_calf_r}
    constraintSetup [{targetBone {boneName=thigh_r}}, {targetBone {boneName=calf_r}}]
    constraintWeights [1,1]
   }
   n16 = "Transform (Modify) Bone - Bone: dyn_volume_calf_l"(ComponentPose=@n15, Alpha=Get knee_rot_l() -> (knee_rot_l:double), Translation="3,3,0", Rotation="0,0,0", Scale="1,1,1.5") {
    rotationMode=BMM_Additive
    rotationSpace=BCS_BoneSpace
    scaleMode=BMM_Additive
    scaleSpace=BCS_BoneSpace
    translationMode=BMM_Additive
    translationSpace=BCS_BoneSpace
    alphaBoolBlend {blendOption=Linear}
    alphaScaleBiasClamp {bMapRange=true, inRange {max=80}}
    boneToModify {boneName=dyn_volume_calf_l}
    scale {z=1.5}
    translation=<3 3>
   }
   n17 = "Transform (Modify) Bone - Bone: dyn_volume_calf_r"(ComponentPose=@n16, Alpha=Get knee_rot_r() -> (knee_rot_r:double), Translation="-3,-3,0", Rotation="0,0,0", Scale="1,1,1.5") {
    rotationMode=BMM_Additive
    rotationSpace=BCS_BoneSpace
    scaleMode=BMM_Additive
    scaleSpace=BCS_BoneSpace
    translationMode=BMM_Additive
    translationSpace=BCS_BoneSpace
    alphaBoolBlend {blendOption=Linear}
    alphaScaleBiasClamp {bMapRange=true, inRange {max=80}}
    boneToModify {boneName=dyn_volume_calf_r}
    scale {z=1.5}
    translation=<-3 -3>
   }
   n18 = Look At - Bone: dyn_pack_aim_l(ComponentPose=@n17, Alpha=1) {
    bUseLookUpAxis=true
    alphaBoolBlend {blendOption=Linear}
    boneToModify {boneName=dyn_pack_aim_l}
    lookAtLocation {x=0}
    lookAtTarget {boneReference {boneName=upperarm_l}}
    lookAt_Axis {axis=<1 0>}
    lookUp_Axis {axis {x=0, z=1}}
   }
   n19 = "Constraint - dyn_pack_base_l (T)"(ComponentPose=@n18, Alpha=0, ConstraintWeights_0=1) {
    alphaBoolBlend {blendOption=Linear}
    boneToModify {boneName=dyn_pack_base_l}
    constraintSetup [{transformType=Translation, targetBone {boneName=upperarm_l}}]
    constraintWeights [1]
   }
   n20 = Look At - Bone: dyn_pack_aim_r(ComponentPose=@n19, Alpha=1) {
    bUseLookUpAxis=true
    alphaBoolBlend {blendOption=Linear}
    boneToModify {boneName=dyn_pack_aim_r}
    lookAtLocation {x=0}
    lookAtTarget {boneReference {boneName=upperarm_r}}
    lookAt_Axis {axis=<-1 0>}
    lookUp_Axis {axis {x=0, z=1}}
   }
   n21 = "Constraint - dyn_pack_base_r (T)"(ComponentPose=@n20, Alpha=0, ConstraintWeights_0=1) {
    alphaBoolBlend {blendOption=Linear}
    boneToModify {boneName=dyn_pack_base_r}
    constraintSetup [{transformType=Translation, targetBone {boneName=upperarm_r}}]
    constraintWeights [1]
   }
   n22 = "Constraint - dyn_base_latt_base_l (PPPP)"!(
    ComponentPose=@n21
    Alpha=1
    ConstraintWeights_0=0.2
    ConstraintWeights_1:float=0.35
    ConstraintWeights_2:float=0.3
    ConstraintWeights_3:float=0.15
   ) -> (Pose) {
    alphaBoolBlend {blendOption=Linear}
    boneToModify {boneName=dyn_base_latt_base_l}
    constraintSetup [{targetBone {boneName=clavicle_l}}, {targetBone {boneName=spine_04}}, {targetBone {boneName=spine_03}}, {targetBone {boneName=spine_02}}]
    constraintWeights [1,1,1,1]
   }
   n23 = "Constraint - dyn_base_ratt_base_r (PPPP)"!(
    ComponentPose=@n22
    Alpha=1
    ConstraintWeights_0=0.2
    ConstraintWeights_1:float=0.35
    ConstraintWeights_2:float=0.3
    ConstraintWeights_3:float=0.15
   ) -> (Pose) {
    alphaBoolBlend {blendOption=Linear}
    boneToModify {boneName=dyn_base_ratt_base_r}
    constraintSetup [{targetBone {boneName=clavicle_r}}, {targetBone {boneName=spine_04}}, {targetBone {boneName=spine_03}}, {targetBone {boneName=spine_02}}]
    constraintWeights [1,1,1,1]
   }
   n24 = "Constraint - dyn_shldr_bk_base_l (PPPP)"!(
    ComponentPose=@n23
    Alpha=1
    ConstraintWeights_0=0.5
    ConstraintWeights_1:float=0.3
    ConstraintWeights_2:float=0.1
    ConstraintWeights_3:float=0.1
   ) -> (Pose) {
    alphaBoolBlend {blendOption=Linear}
    boneToModify {boneName=dyn_shldr_bk_base_l}
    constraintSetup [
     {targetBone {boneName=upperarm_twist_01_l}}
     {targetBone {boneName=clavicle_l}}
     {targetBone {boneName=spine_04}}
     {targetBone {boneName=spine_03}}
    ]
    constraintWeights [1,1,1,1]
   }
   n25 = "Constraint - dyn_shldr_bk_base_r (PPPP)"!(
    ComponentPose=@n24
    Alpha=1
    ConstraintWeights_0=0.5
    ConstraintWeights_1:float=0.3
    ConstraintWeights_2:float=0.1
    ConstraintWeights_3:float=0.1
   ) -> (Pose) {
    alphaBoolBlend {blendOption=Linear}
    boneToModify {boneName=dyn_shldr_bk_base_r}
    constraintSetup [
     {targetBone {boneName=upperarm_twist_01_r}}
     {targetBone {boneName=clavicle_r}}
     {targetBone {boneName=spine_04}}
     {targetBone {boneName=spine_03}}
    ]
    constraintWeights [1,1,1,1]
   }
   n26 = Get l_arm_fwd() -> (l_arm_fwd:double)
   n27 = "Transform (Modify) Bone - Bone: dyn_pack_l"(ComponentPose=@n25, Alpha=@n26, Translation="-15,0,0", Rotation="0,0,0", Scale="1,1,1") {
    rotationMode=BMM_Additive
    rotationSpace=BCS_BoneSpace
    scaleMode=BMM_Additive
    scaleSpace=BCS_BoneSpace
    translationMode=BMM_Additive
    translationSpace=BCS_BoneSpace
    alphaBoolBlend {blendOption=Linear}
    alphaScaleBiasClamp {inRange {max=0, min=-100}, outRange {max=0, min=1}}
    boneToModify {boneName=dyn_pack_l}
    translation {x=-15}
   }
   n28 = "Transform (Modify) Bone - Bone: dyn_bicep_fr_l"(ComponentPose=@n27, Alpha=@n26, Translation="0,-10,0", Rotation="0,0,0", Scale="1,1,1") {
    rotationMode=BMM_Additive
    rotationSpace=BCS_BoneSpace
    scaleMode=BMM_Additive
    scaleSpace=BCS_BoneSpace
    translationMode=BMM_Additive
    translationSpace=BCS_BoneSpace
    alphaBoolBlend {blendOption=Linear}
    alphaScaleBiasClamp {inRange {max=0, min=-100}, outRange {max=0, min=1}}
    boneToModify {boneName=dyn_bicep_fr_l}
    translation {y=-10}
   }
   n29 = "Transform (Modify) Bone - Bone: dyn_shldr_bk_l"(ComponentPose=@n28, Alpha=@n26, Translation="0,-3,0", Rotation="0,0,0", Scale="1,1,1") {
    rotationMode=BMM_Additive
    rotationSpace=BCS_BoneSpace
    scaleMode=BMM_Additive
    scaleSpace=BCS_BoneSpace
    translationMode=BMM_Additive
    translationSpace=BCS_BoneSpace
    alphaBoolBlend {blendOption=Linear}
    alphaScaleBiasClamp {inRange {max=0, min=-100}, outRange {max=0, min=1}}
    boneToModify {boneName=dyn_shldr_bk_l}
    translation {y=-3}
   }
   n30 = Get l_arm_up() -> (l_arm_up:double)
   n31 = "Transform (Modify) Bone - Bone: dyn_pack_l"(ComponentPose=@n29, Alpha=@n30, Translation="10,0,0", Rotation="0,0,0", Scale="1,1,1") {
    rotationMode=BMM_Additive
    rotationSpace=BCS_BoneSpace
    scaleMode=BMM_Additive
    scaleSpace=BCS_BoneSpace
    translationMode=BMM_Additive
    translationSpace=BCS_BoneSpace
    alphaBoolBlend {blendOption=Linear}
    alphaScaleBiasClamp {inRange {max=0, min=-100}, outRange {max=0, min=1}}
    boneToModify {boneName=dyn_pack_l}
    translation {x=10}
   }
   n32 = "Transform (Modify) Bone - Bone: dyn_bicep_fr_l"(ComponentPose=@n31, Alpha=@n30, Translation="0,6,0", Rotation="0,0,0", Scale="1,1,1") {
    rotationMode=BMM_Additive
    rotationSpace=BCS_BoneSpace
    scaleMode=BMM_Additive
    scaleSpace=BCS_BoneSpace
    translationMode=BMM_Additive
    translationSpace=BCS_BoneSpace
    alphaBoolBlend {blendOption=Linear}
    alphaScaleBiasClamp {inRange {max=0, min=-100}, outRange {max=0, min=1}}
    boneToModify {boneName=dyn_bicep_fr_l}
    translation {y=6}
   }
   n33 = Get l_arm_dn() -> (l_arm_dn:double)
   n34 = "Transform (Modify) Bone - Bone: dyn_latt_l"(ComponentPose=@n32, Alpha=@n33, Translation="0,0,7", Rotation="0,0,0", Scale="1,1,1") {
    rotationMode=BMM_Additive
    rotationSpace=BCS_BoneSpace
    scaleMode=BMM_Additive
    scaleSpace=BCS_BoneSpace
    translationMode=BMM_Additive
    translationSpace=BCS_BoneSpace
    alphaBoolBlend {blendOption=Linear}
    alphaScaleBiasClamp {inRange {max=0, min=-100}, outRange {max=0, min=1}}
    boneToModify {boneName=dyn_latt_l}
    translation {z=7}
   }
   n35 = "Transform (Modify) Bone - Bone: upperarm_twist_02_l"(ComponentPose=@n34, Alpha=@n33, Translation="0,0,0", Rotation="0,0,0", Scale="1,1,0.8") {
    rotationMode=BMM_Additive
    rotationSpace=BCS_BoneSpace
    scaleMode=BMM_Additive
    scaleSpace=BCS_BoneSpace
    translationMode=BMM_Additive
    translationSpace=BCS_BoneSpace
    alphaBoolBlend {blendOption=Linear}
    alphaScaleBiasClamp {inRange {max=0, min=-100}, outRange {max=0, min=1}}
    boneToModify {boneName=upperarm_twist_02_l}
    scale {z=0.8}
   }
   n36 = Get r_arm_fwd() -> (r_arm_fwd:double)
   n37 = "Transform (Modify) Bone - Bone: dyn_pack_r"(ComponentPose=@n35, Alpha=@n36, Translation="15,0,0", Rotation="0,0,0", Scale="1,1,1") {
    rotationMode=BMM_Additive
    rotationSpace=BCS_BoneSpace
    scaleMode=BMM_Additive
    scaleSpace=BCS_BoneSpace
    translationMode=BMM_Additive
    translationSpace=BCS_BoneSpace
    alphaBoolBlend {blendOption=Linear}
    alphaScaleBiasClamp {inRange {max=0, min=-100}, outRange {max=0, min=1}}
    boneToModify {boneName=dyn_pack_r}
    translation {x=15}
   }
   n38 = "Transform (Modify) Bone - Bone: dyn_bicep_fr_r"(ComponentPose=@n37, Alpha=@n36, Translation="0,10,0", Rotation="0,0,0", Scale="1,1,1") {
    rotationMode=BMM_Additive
    rotationSpace=BCS_BoneSpace
    scaleMode=BMM_Additive
    scaleSpace=BCS_BoneSpace
    translationMode=BMM_Additive
    translationSpace=BCS_BoneSpace
    alphaBoolBlend {blendOption=Linear}
    alphaScaleBiasClamp {inRange {max=0, min=-100}, outRange {max=0, min=1}}
    boneToModify {boneName=dyn_bicep_fr_r}
    translation {y=10}
   }
   n39 = "Transform (Modify) Bone - Bone: dyn_shldr_bk_r"(ComponentPose=@n38, Alpha=@n36, Translation="0,3,0", Rotation="0,0,0", Scale="1,1,1") {
    rotationMode=BMM_Additive
    rotationSpace=BCS_BoneSpace
    scaleMode=BMM_Additive
    scaleSpace=BCS_BoneSpace
    translationMode=BMM_Additive
    translationSpace=BCS_BoneSpace
    alphaBoolBlend {blendOption=Linear}
    alphaScaleBiasClamp {inRange {max=0, min=-100}, outRange {max=0, min=1}}
    boneToModify {boneName=dyn_shldr_bk_r}
    translation {y=3}
   }
   n40 = Get r_arm_up() -> (r_arm_up:double)
   n41 = "Transform (Modify) Bone - Bone: dyn_pack_r"(ComponentPose=@n39, Alpha=@n40, Translation="-10,0,0", Rotation="0,0,0", Scale="1,1,1") {
    rotationMode=BMM_Additive
    rotationSpace=BCS_BoneSpace
    scaleMode=BMM_Additive
    scaleSpace=BCS_BoneSpace
    translationMode=BMM_Additive
    translationSpace=BCS_BoneSpace
    alphaBoolBlend {blendOption=Linear}
    alphaScaleBiasClamp {inRange {max=0, min=-100}, outRange {max=0, min=1}}
    boneToModify {boneName=dyn_pack_r}
    translation {x=-10}
   }
   n42 = "Transform (Modify) Bone - Bone: dyn_bicep_fr_r"(ComponentPose=@n41, Alpha=@n40, Translation="0,-6,0", Rotation="0,0,0", Scale="1,1,1") {
    rotationMode=BMM_Additive
    rotationSpace=BCS_BoneSpace
    scaleMode=BMM_Additive
    scaleSpace=BCS_BoneSpace
    translationMode=BMM_Additive
    translationSpace=BCS_BoneSpace
    alphaBoolBlend {blendOption=Linear}
    alphaScaleBiasClamp {inRange {max=0, min=-100}, outRange {max=0, min=1}}
    boneToModify {boneName=dyn_bicep_fr_r}
    translation {y=-6}
   }
   n43 = Get r_arm_dn() -> (r_arm_dn:double)
   n44 = "Transform (Modify) Bone - Bone: dyn_ratt_r"(ComponentPose=@n42, Alpha=@n43, Translation="0,0,-7", Rotation="0,0,0", Scale="1,1,1") {
    rotationMode=BMM_Additive
    rotationSpace=BCS_BoneSpace
    scaleMode=BMM_Additive
    scaleSpace=BCS_BoneSpace
    translationMode=BMM_Additive
    translationSpace=BCS_BoneSpace
    alphaBoolBlend {blendOption=Linear}
    alphaScaleBiasClamp {inRange {max=0, min=-100}, outRange {max=0, min=1}}
    boneToModify {boneName=dyn_ratt_r}
    translation {z=-7}
   }
   n45 = "Transform (Modify) Bone - Bone: upperarm_twist_02_r"(ComponentPose=@n44, Alpha=@n43, Translation="0,0,0", Rotation="0,0,0", Scale="1,1,0.8") {
    rotationMode=BMM_Additive
    rotationSpace=BCS_BoneSpace
    scaleMode=BMM_Additive
    scaleSpace=BCS_BoneSpace
    translationMode=BMM_Additive
    translationSpace=BCS_BoneSpace
    alphaBoolBlend {blendOption=Linear}
    alphaScaleBiasClamp {inRange {max=0, min=-100}, outRange {max=0, min=1}}
    boneToModify {boneName=upperarm_twist_02_r}
    scale {z=0.8}
   }
   n46 = Get r_pistonAimUp_alpha() -> (r_pistonAimUp_alpha:double)
   n47 = "Transform (Modify) Bone - Bone: upperarm_twist_01_r"(ComponentPose=@n45, Alpha=@n46, Translation="3,0,0", Rotation="-30,0,0", Scale="1,1,1") {
    rotationMode=BMM_Additive
    rotationSpace=BCS_BoneSpace
    scaleMode=BMM_Additive
    scaleSpace=BCS_BoneSpace
    translationMode=BMM_Additive
    translationSpace=BCS_BoneSpace
    alphaBoolBlend {blendOption=Linear}
    alphaScaleBiasClamp {inRange {max=0, min=-100}, outRange {max=0, min=1}}
    boneToModify {boneName=upperarm_twist_01_r}
    rotation {pitch=-30}
    translation {x=3}
   }
   n48 = "Transform (Modify) Bone - Bone: upperarm_twist_02_r"(ComponentPose=@n47, Alpha=@n46, Translation="0,0,-5", Rotation="0,0,0", Scale="1,1,1") {
    rotationMode=BMM_Additive
    rotationSpace=BCS_BoneSpace
    scaleMode=BMM_Additive
    scaleSpace=BCS_BoneSpace
    translationMode=BMM_Additive
    translationSpace=BCS_BoneSpace
    alphaBoolBlend {blendOption=Linear}
    alphaScaleBiasClamp {inRange {max=0, min=-100}, outRange {max=0, min=1}}
    boneToModify {boneName=upperarm_twist_02_r}
    translation {z=-5}
   }
   "Transform (Modify) Bone - Bone: dyn_bicep_bk_r"(ComponentPose=@n48, Alpha=@n46, Translation="0,-2,0", Rotation="0,0,0", Scale="1,1,1") {
    rotationMode=BMM_Additive
    rotationSpace=BCS_BoneSpace
    scaleMode=BMM_Additive
    scaleSpace=BCS_BoneSpace
    translationMode=BMM_Additive
    translationSpace=BCS_BoneSpace
    alphaBoolBlend {blendOption=Linear}
    alphaScaleBiasClamp {inRange {max=0, min=-100}, outRange {max=0, min=1}}
    boneToModify {boneName=dyn_bicep_bk_r}
    translation {y=-2}
   }
   n49 = Control Rig(Source=Copy Pose From Mesh() {bUseAttachedParent=true}) {
    alphaBoolBlend {blendOption=Linear}
    controlRigAssetReference {blueprintRigClass=&/M_MED_Quarrel_Deform_CtrlRig_twists.M_MED_Quarrel_Deform_CtrlRig_twists_C}
   }
   n50 = Control Rig!(
    Source=@n49
    Alpha
    KneeRot_fr_l_114deg:Vector
    KneeTrans_fr_l_114deg:Vector="0.258,-3.591,0.009"
    KneeScale_fr_l_114deg:Vector="1,1,1"
    KneeScale_fr_l_130deg:Vector="1,1,1"
    KneeRot_fr_l_130deg:Vector="0,0,-51.2"
    KneeTrans_fr_l_130deg:Vector="0.258,-3.591,0.009"
   ) -> (Pose) {
    alphaBoolBlend {blendOption=Linear}
    controlRigAssetReference {blueprintRigClass=&/M_MED_Quarrel_Deform_CtrlRig_sculpters.M_MED_Quarrel_Deform_CtrlRig_sculpters_C}
   }
   n51 = "Transform (Modify) Bone - Bone: dyn_bicep_bk_l"(ComponentPose=Local To Component(LocalPose=@n50), Alpha=Get l_arm_fwd(), Translation="0,10,0", Rotation="0,0,0", Scale="1,1,1") {
    rotationMode=BMM_Additive
    rotationSpace=BCS_BoneSpace
    scaleMode=BMM_Additive
    scaleSpace=BCS_BoneSpace
    translationMode=BMM_Additive
    translationSpace=BCS_BoneSpace
    alphaBoolBlend {blendOption=Linear}
    alphaScaleBiasClamp {inRange {max=0, min=-100}, outRange {max=0, min=1}}
    boneToModify {boneName=dyn_bicep_bk_l}
    translation {y=10}
   }
   n52 = "Transform (Modify) Bone - Bone: dyn_bicep_bk_r"(ComponentPose=@n51, Alpha=Get r_arm_fwd(), Translation="0,-10,0", Rotation="0,0,0", Scale="1,1,1") {
    rotationMode=BMM_Additive
    rotationSpace=BCS_BoneSpace
    scaleMode=BMM_Additive
    scaleSpace=BCS_BoneSpace
    translationMode=BMM_Additive
    translationSpace=BCS_BoneSpace
    alphaBoolBlend {blendOption=Linear}
    alphaScaleBiasClamp {inRange {max=0, min=-100}, outRange {max=0, min=1}}
    boneToModify {boneName=dyn_bicep_bk_r}
    translation {y=-10}
   }
   n53 = "Transform (Modify) Bone - Bone: def_belt_l"(ComponentPose=@n52, Alpha=Get thigh_rot_l(), Translation="10,0,1", Rotation="0,-15,0", Scale="1,1,1") {
    rotationMode=BMM_Additive
    rotationSpace=BCS_BoneSpace
    translationMode=BMM_Additive
    translationSpace=BCS_BoneSpace
    alphaBoolBlend {blendOption=Linear}
    boneToModify {boneName=def_belt_l}
    rotation {yaw=-15}
    translation {x=10, z=1}
   }
   n54 = "Transform (Modify) Bone - Bone: def_belt_r"(ComponentPose=@n53, Alpha=Get thigh_rot_r(), Translation="10,0,0", Rotation="0,10,0", Scale="1,1,1") {
    rotationMode=BMM_Additive
    rotationSpace=BCS_BoneSpace
    translationMode=BMM_Additive
    translationSpace=BCS_BoneSpace
    alphaBoolBlend {blendOption=Linear}
    boneToModify {boneName=def_belt_r}
    rotation {yaw=10}
    translation {x=10}
   }
   n55 = "Constraint - dyn_pouches_01_MSTR (PP)"!(ComponentPose=@n54, Alpha=1, ConstraintWeights_0=0.7437, ConstraintWeights_1:float=0.2563) -> (Pose) {
    alphaBoolBlend {blendOption=Linear}
    boneToModify {boneName=dyn_pouches_01_MSTR}
    constraintSetup [{targetBone {boneName=pelvis}}, {targetBone {boneName=def_belt_l}}]
    constraintWeights [1,1]
   }
   n56 = "Constraint - dyn_pouches_02_MSTR (PP)"!(ComponentPose=@n55, Alpha=1, ConstraintWeights_0=0.8928, ConstraintWeights_1:float=0.1072) -> (Pose) {
    alphaBoolBlend {blendOption=Linear}
    boneToModify {boneName=dyn_pouches_02_MSTR}
    constraintSetup [{targetBone {boneName=pelvis}}, {targetBone {boneName=def_belt_l}}]
    constraintWeights [1,1]
   }
   n57 = Get thigh_rot_l()
   n58 = "Transform (Modify) Bone - Bone: dyn_pouches_01"(ComponentPose=@n56, Alpha=@n57, Translation="1.393,-3.548,-0.213", Rotation="5.000,-15.003,-40.000", Scale="1,1,1") {
    rotationMode=BMM_Additive
    rotationSpace=BCS_BoneSpace
    translationMode=BMM_Additive
    translationSpace=BCS_BoneSpace
    alphaBoolBlend {blendOption=Linear}
    boneToModify {boneName=dyn_pouches_01}
    rotation=<rot 5 -15.003 -40>
    translation=<1.3929 -3.548 -0.213>
   }
   n59 = "Transform (Modify) Bone - Bone: dyn_pouches_02"(ComponentPose=@n58, Alpha=@n57, Translation="-0.144,-3.241,-0.206", Rotation="0.000,0.000,-10.004", Scale="1,1,1") {
    rotationMode=BMM_Additive
    rotationSpace=BCS_BoneSpace
    translationMode=BMM_Additive
    translationSpace=BCS_BoneSpace
    alphaBoolBlend {blendOption=Linear}
    boneToModify {boneName=dyn_pouches_02}
    rotation {roll=-10.0041}
    translation=<-0.144 -3.241 -0.2062>
   }
   n60 = "Constraint - dyn_shoelaces_l_MSTR (PP)"!(ComponentPose=@n59, Alpha=1, ConstraintWeights_0=0.4444, ConstraintWeights_1:float=0.5556) -> (Pose) {
    alphaBoolBlend {blendOption=Linear}
    boneToModify {boneName=dyn_shoelaces_l_MSTR}
    constraintSetup [{targetBone {boneName=calf_l}}, {targetBone {boneName=calf_twist_02_l}}]
    constraintWeights [1,1]
   }
   n61 = "Constraint - dyn_shoelaces_r_MSTR (PP)"!(ComponentPose=@n60, Alpha=1, ConstraintWeights_0=0.5185, ConstraintWeights_1:float=0.4815) -> (Pose) {
    alphaBoolBlend {blendOption=Linear}
    boneToModify {boneName=dyn_shoelaces_r_MSTR}
    constraintSetup [{targetBone {boneName=calf_twist_02_r}}, {targetBone {boneName=calf_r}}]
    constraintWeights [1,1]
   }
   n62 = Anim Dynamics - Bone: dyn_shoelaces_l(ComponentPose=@n61, Alpha=1, ExternalForce="0,0,0") {
    angularDampingOverride=0.995
    bGravityOverrideInSimSpace=true
    bOverrideAngularDamping=true
    bUseGravityOverride=true
    simulationSpace=RootRelative
    alphaBoolBlend {blendOption=Linear}
    boundBone {boneName=dyn_shoelaces_l}
    componentAppliedLinearAccClamp=<3000 3000 3000>
    componentLinearAccScale=<0.3 0.3 0.3>
    componentLinearVelScale=<0.3 0.3 0.3>
    gravityOverride {z=-980}
    physicsBodyDefinitions [{
      sphereCollisionRadius=0
      boundBone {boneName=dyn_shoelaces_l}
      boxExtents=<10 8 8>
      constraintSetup {angularLimitsMax=<0 2 45>, angularLimitsMin=<0 -45 -45>, angularTarget=<0 0 0>}
      localJointOffset=<1.1 0 0>
     }]
    relativeSpaceBone {boneName=calf_l}
   }
   n63 = Anim Dynamics - Bone: dyn_shoelaces_r(ComponentPose=@n62, Alpha=1, ExternalForce="0,0,0") {
    angularDampingOverride=0.999
    bGravityOverrideInSimSpace=true
    bOverrideAngularDamping=true
    bUseGravityOverride=true
    simulationSpace=RootRelative
    alphaBoolBlend {blendOption=Linear}
    boundBone {boneName=dyn_shoelaces_r}
    componentAppliedLinearAccClamp=<3000 3000 3000>
    componentLinearAccScale=<0.3 0.3 0.3>
    componentLinearVelScale=<0.3 0.3 0.3>
    gravityOverride {z=-980}
    physicsBodyDefinitions [{
      sphereCollisionRadius=0
      boundBone {boneName=dyn_shoelaces_r}
      boxExtents=<10 8 8>
      constraintSetup {angularLimitsMax=<0 2 45>, angularLimitsMin=<0 -45 -45>, angularTarget=<0 0 0>}
      localJointOffset=<1 0 0>
     }]
    relativeSpaceBone {boneName=calf_r}
   }
   n64 = "Constraint - dyn_cuff_01_MSTR (PP)"!(ComponentPose=@n63, Alpha=1, ConstraintWeights_0=0.6983, ConstraintWeights_1:float=0.3017) -> (Pose) {
    alphaBoolBlend {blendOption=Linear}
    boneToModify {boneName=dyn_cuff_01_MSTR}
    constraintSetup [{targetBone {boneName=pelvis}}, {targetBone {boneName=def_belt_r}}]
    constraintWeights [1,1]
   }
   n65 = "Constraint - dyn_cuff_02_MSTR (PP)"!(ComponentPose=@n64, Alpha=1, ConstraintWeights_0=0.872, ConstraintWeights_1:float=0.128) -> (Pose) {
    alphaBoolBlend {blendOption=Linear}
    boneToModify {boneName=dyn_cuff_02_MSTR}
    constraintSetup [{targetBone {boneName=pelvis}}, {targetBone {boneName=def_belt_r}}]
    constraintWeights [1,1]
   }
   n66 = "Transform (Modify) Bone - Bone: dyn_cuff_01"(ComponentPose=@n65, Alpha=Get thigh_rot_r(), Translation="-0.550,1.507,-0.738", Rotation="-2.498,64.908,-0.668", Scale="1,1,1") {
    rotationMode=BMM_Additive
    rotationSpace=BCS_BoneSpace
    translationMode=BMM_Additive
    translationSpace=BCS_BoneSpace
    alphaBoolBlend {blendOption=Linear}
    boneToModify {boneName=dyn_cuff_01}
    rotation=<rot -2.498 64.9084 -0.668>
    translation=<-0.55 1.507 -0.738>
   }
   n67 = Get thigh_rot_out_r() -> (thigh_rot_out_r:double)
   n68 = "Transform (Modify) Bone - Bone: dyn_cuff_02"(ComponentPose=@n66, Alpha=@n67, Translation="-1.098,-0.294,0.000", Rotation="0.000,15.000,0.000", Scale="1,1,1") {
    rotationMode=BMM_Additive
    rotationSpace=BCS_BoneSpace
    translationMode=BMM_Additive
    translationSpace=BCS_BoneSpace
    alphaBoolBlend {blendOption=Linear}
    boneToModify {boneName=dyn_cuff_02}
    rotation {yaw=15.0001}
    translation=<-1.098 -0.294>
   }
   n69 = "Transform (Modify) Bone - Bone: dyn_cuff_01"(ComponentPose=@n68, Alpha=@n67, Translation="0,0,0", Rotation="0.000,25,0.000", Scale="1,1,1") {
    rotationMode=BMM_Additive
    rotationSpace=BCS_BoneSpace
    translationMode=BMM_Additive
    translationSpace=BCS_BoneSpace
    alphaBoolBlend {blendOption=Linear}
    boneToModify {boneName=dyn_cuff_01}
    rotation {yaw=25}
   }
   n70 = Get thigh_rot_out_r()
   n71 = "Constraint - dyn_cuffChain_01_AIM (PP)"!(ComponentPose=@n69, Alpha=@n70, ConstraintWeights_0=1, ConstraintWeights_1:float=0) -> (Pose) {
    alphaBoolBlend {blendOption=Linear}
    boneToModify {boneName=dyn_cuffChain_01_AIM}
    constraintSetup [{targetBone {boneName=thigh_twist_01_r}}, {targetBone {boneName=pelvis}}]
    constraintWeights [1,1]
   }
   n72 = "Constraint - dyn_cuffChain_02_AIM (PP)"!(ComponentPose=@n71, Alpha=@n70, ConstraintWeights_0=1, ConstraintWeights_1:float=0) -> (Pose) {
    alphaBoolBlend {blendOption=Linear}
    boneToModify {boneName=dyn_cuffChain_02_AIM}
    constraintSetup [{targetBone {boneName=thigh_twist_01_r}}, {targetBone {boneName=pelvis}}]
    constraintWeights [1,1]
   }
   n73 = "Transform (Modify) Bone - Bone: dyn_cuffChain_01_AIM"(ComponentPose=@n72, Alpha=@n70, Translation="0,0,20", Rotation="0,0,0", Scale="1,1,1") {
    rotationMode=BMM_Additive
    rotationSpace=BCS_BoneSpace
    translationMode=BMM_Additive
    translationSpace=BCS_BoneSpace
    alphaBoolBlend {blendOption=Linear}
    boneToModify {boneName=dyn_cuffChain_01_AIM}
    translation {z=20}
   }
   n74 = "Transform (Modify) Bone - Bone: dyn_cuffChain_02_AIM"(ComponentPose=@n73, Alpha=@n70, Translation="0,0,20", Rotation="0,0,0", Scale="1,1,1") {
    rotationMode=BMM_Additive
    rotationSpace=BCS_BoneSpace
    translationMode=BMM_Additive
    translationSpace=BCS_BoneSpace
    alphaBoolBlend {blendOption=Linear}
    boneToModify {boneName=dyn_cuffChain_02_AIM}
    translation {z=20}
   }
   n75 = Look At - Bone: dyn_cuffChain_01_MSTR(ComponentPose=@n74, Alpha=1) {
    alphaBoolBlend {blendOption=Linear}
    boneToModify {boneName=dyn_cuffChain_01_MSTR}
    lookAtLocation {x=0}
    lookAtTarget {boneReference {boneName=dyn_cuffChain_01_AIM}}
    lookAt_Axis {axis=<1 0>}
   }
   n76 = Look At - Bone: dyn_cuffChain_02_MSTR(ComponentPose=@n75, Alpha=1) {
    alphaBoolBlend {blendOption=Linear}
    boneToModify {boneName=dyn_cuffChain_02_MSTR}
    lookAtLocation {x=0}
    lookAtTarget {boneReference {boneName=dyn_cuffChain_02_AIM}}
    lookAt_Axis {axis=<1 0>}
   }
   n77 = "Constraint - dyn_comb_MSTR (PP)"!(ComponentPose=@n76, Alpha=1, ConstraintWeights_0=0.9112, ConstraintWeights_1:float=0.0888) -> (Pose) {
    alphaBoolBlend {blendOption=Linear}
    boneToModify {boneName=dyn_comb_MSTR}
    constraintSetup [{targetBone {boneName=pelvis}}, {targetBone {boneName=thigh_twist_01_r}}]
    constraintWeights [1,1]
   }
   n78 = RigidBody!(ComponentPose=@n77, Alpha, OverrideWorldGravity:Vector="0,0,-980", ExternalForce) -> (Pose) {
    bClampLinearTranslationLimitToRefPose=true
    bOverrideWorldGravity=true
    overridePhysicsAsset=&/M_MED_Quarrel_dyn_necklace
    simulationSpace=BaseBoneSpace
    alphaBoolBlend {blendOption=Linear}
    baseBoneRef {boneName=spine_05}
    componentAppliedLinearAccClamp=<2000 2000 1000>
    componentLinearAccScale {x=0.1}
    componentLinearVelScale {x=0.1}
    overrideWorldGravity {z=-980}
    simSpaceSettings {maxAngularAcceleration=50, maxAngularVelocity=2, maxLinearAcceleration=50, maxLinearVelocity=5, velocityScaleZ=0, worldAlpha=1}
   }
   n79 = Anim Dynamics - Bone: dyn_dogTag_01(ComponentPose=@n78, Alpha=1, ExternalForce="0,0,0") {
    angularDampingOverride=0.9999
    bGravityOverrideInSimSpace=true
    bOverrideAngularDamping=true
    bUseGravityOverride=true
    simulationSpace=RootRelative
    alphaBoolBlend {blendOption=Linear}
    boundBone {boneName=dyn_dogTag_01}
    componentAppliedLinearAccClamp=<3000 3000 3000>
    componentLinearAccScale=<0.3 0.3 0.3>
    componentLinearVelScale=<0.3 0.3 0.3>
    gravityOverride {z=-980}
    physicsBodyDefinitions [{
      sphereCollisionRadius=0
      boundBone {boneName=dyn_dogTag_01}
      boxExtents=<10 8 8>
      constraintSetup {angularLimitsMax=<0 0 45>, angularLimitsMin=<0 -25 -45>, angularTarget=<0 0 0>}
      localJointOffset=<1.5 0 0>
     }]
    relativeSpaceBone {boneName=dyn_necklace_02}
   }
   n80 = Anim Dynamics - Bone: dyn_dogTag_02(ComponentPose=@n79, Alpha=1, ExternalForce="0,0,0") {
    angularDampingOverride=0.9999
    bGravityOverrideInSimSpace=true
    bOverrideAngularDamping=true
    bUseGravityOverride=true
    simulationSpace=RootRelative
    alphaBoolBlend {blendOption=Linear}
    boundBone {boneName=dyn_dogTag_02}
    componentAppliedLinearAccClamp=<3000 3000 3000>
    componentLinearAccScale=<0.3 0.3 0.3>
    componentLinearVelScale=<0.3 0.3 0.3>
    gravityOverride {z=-980}
    physicsBodyDefinitions [{
      sphereCollisionRadius=0
      boundBone {boneName=dyn_dogTag_02}
      boxExtents=<10 10 10>
      constraintSetup {angularLimitsMax=<0 0 45>, angularLimitsMin=<0 -10 -45>, angularTarget=<0 0 0>}
      localJointOffset=<1.5 0 0>
     }]
    relativeSpaceBone {boneName=dyn_necklace_02}
   }
   n81 = RigidBody(ComponentPose=@n80, OverrideWorldGravity="0,0,-980") {
    bClampLinearTranslationLimitToRefPose=true
    bOverrideWorldGravity=true
    overridePhysicsAsset=&/M_MED_Quarrel_dyn_cuffs
    simulationSpace=BaseBoneSpace
    alphaBoolBlend {blendOption=Linear}
    baseBoneRef {boneName=pelvis}
    componentAppliedLinearAccClamp=<2000 2000 1000>
    componentLinearAccScale {x=0.2}
    componentLinearVelScale {x=0.2}
    overrideWorldGravity {z=-980}
    simSpaceSettings {maxAngularAcceleration=200, maxAngularVelocity=10, maxLinearAcceleration=500, maxLinearVelocity=50, velocityScaleZ=0, worldAlpha=1}
   }
   Output Pose(Result=Component To Local(ComponentPose=@n81)) {layerGroup=None}
  ]}
 EventGraph {nodes [
   Event BlueprintUpdateAnimation()
   > n1 = Sequence!(execute) -> (then_0, then_1, then_2:Exec, then_3:Exec, then_4:Exec, then_5:Exec, then_6:Exec)
   n2 = Sequence(execute=@n1.then_6)
   n3 = Sequence(execute=@n2.then_3)
   n4 = Sequence(execute=@n3.then_5)
   n5 = Branch(execute=@n4.then_2, Condition=Get bIsInFrontEnd())
   > Set isInFrontEndAlpha(isInFrontEndAlpha:double=1) -> (Output_Get:double)
   Set isInFrontEndAlpha(execute=@n5.else, isInFrontEndAlpha=0)
   n6 = Branch(execute=@n4.then_1, Condition=Get bIsCrouching())
   > Set crouchingAlpha(crouchingAlpha:double=1) -> (Output_Get:double)
   Set crouchingAlpha(execute=@n6.else, crouchingAlpha=0)
   n7 = Branch(execute=@n3.then_1, Condition=Get isPistol() -> (isPistol:bool))
   > n8 = Branch(Condition="float >= float"(A=Get r_arm_fwd(), B=0.8) {is=title K2Node_CallFunction})
   > Set r_pistonAimUp_alpha(r_pistonAimUp_alpha:double=MapRangeClamped(Value=Get r_arm_up(), InRangeA=0.5, InRangeB=1)) -> (Output_Get:double)
   Set r_pistonAimUp_alpha(execute=@n8.else, r_pistonAimUp_alpha=0)
   Set r_pistonAimUp_alpha(execute=@n7.else, r_pistonAimUp_alpha=0)
   n9 = Cast To FortPlayerAnimInstance(execute=@n3.then_0, Object=GetAnimInstance(self=GetOwnerSkeletalMeshComponent()))
   > Set isPistol(isPistol="Equal (Enum)"(A:EFortWeaponCoreAnimation Enum=Get WeaponCoreAnim(self=@n9), B:EFortWeaponCoreAnimation Enum=Pistol))
   n10 = GetPartSkeletalMeshComponent()
   Set l_arm_fwd(
    execute=@n2.then_0
    l_arm_fwd:double=GetAlphaFromReferenceTransformDelta(
     SkeletalMeshComponent=@n10
     BoneName=lowerarm_l
     BaseName=dyn_arm_measure_l
     DeltaComponentType=TranslationY
     InputRange=(X=5,Y=25)
     TwistAxis="1,0,0"
    )
   ) -> (Output_Get:double)
   n11 = GetPartSkeletalMeshComponent()
   > Set r_arm_fwd(r_arm_fwd:double=GetAlphaFromReferenceTransformDelta(
     SkeletalMeshComponent=@n11
     BoneName=lowerarm_r
     BaseName=dyn_arm_measure_r
     DeltaComponentType=TranslationY
     InputRange=(X=-25,Y=-5)
     OutputRange=(X=1,Y=0)
     TwistAxis="1,0,0"
    )) -> (Output_Get:double)
   Set l_arm_up(
    execute=@n2.then_1
    l_arm_up:double=GetAlphaFromReferenceTransformDelta(
     SkeletalMeshComponent=@n10
     BoneName=lowerarm_l
     BaseName=dyn_arm_measure_l
     DeltaComponentType=TranslationZ
     InputRange=(X=10,Y=25)
     TwistAxis="1,0,0"
    )
   ) -> (Output_Get:double)
   > Set r_arm_up(r_arm_up:double=GetAlphaFromReferenceTransformDelta(
     SkeletalMeshComponent=@n11
     BoneName=lowerarm_r
     BaseName=dyn_arm_measure_r
     DeltaComponentType=TranslationZ
     InputRange=(X=-25,Y=-10)
     OutputRange=(X=1,Y=0)
     TwistAxis="1,0,0"
    )) -> (Output_Get:double)
   Set l_arm_dn(
    execute=@n2.then_2
    l_arm_dn:double=GetAlphaFromReferenceTransformDelta(
     SkeletalMeshComponent=@n10
     BoneName=lowerarm_l
     BaseName=dyn_arm_measure_l
     DeltaComponentType=TranslationZ
     InputRange=(X=-7,Y=0)
     OutputRange=(X=1,Y=0)
     TwistAxis="1,0,0"
    )
   ) -> (Output_Get:double)
   > Set r_arm_dn(r_arm_dn:double=GetAlphaFromReferenceTransformDelta(
     SkeletalMeshComponent=@n11
     BoneName=lowerarm_r
     BaseName=dyn_arm_measure_r
     DeltaComponentType=TranslationZ
     InputRange=(X=0,Y=7)
     TwistAxis="1,0,0"
    )) -> (Output_Get:double)
   n12 = GetDeltaTransformFromRefPose(self=GetOwnerSkeletalMeshComponent(), BoneName=lowerarm_l, BaseName=upperarm_l)
   Set elbow_rot_l(execute=@n1.then_0, elbow_rot_l:double=BreakRotator(InRot=BreakTransform(InTransform=@n12).Rotation).Yaw) -> (Output_Get:double)
   n13 = GetOwnerSkeletalMeshComponent()
   n14 = BreakRotator(InRot=BreakTransform(InTransform=GetDeltaTransformFromRefPose(self=@n13, BoneName=thigh_r, BaseName=pelvis)).Rotation)
   Set thigh_rot_r(execute=@n1.then_3, thigh_rot_r=MapRangeClamped(Value=@n14.Yaw, InRangeA=0, InRangeB=-90))
   > Set thigh_rot_out_r(thigh_rot_out_r:double=MapRangeClamped(Value=@n14.Pitch, InRangeA=0, InRangeB=60)) -> (Output_Get:double)
   n15 = GetDeltaTransformFromRefPose(self=GetOwnerSkeletalMeshComponent(), BoneName=calf_r, BaseName=thigh_r)
   Set knee_rot_r(execute=@n1.then_5, knee_rot_r:double=BreakRotator(InRot=BreakTransform(InTransform=@n15).Rotation).Yaw) -> (Output_Get:double)
   n16 = GetDeltaTransformFromRefPose(self=GetOwnerSkeletalMeshComponent(), BoneName=calf_l, BaseName=thigh_l)
   Set knee_rot_l(execute=@n1.then_4, knee_rot_l:double=BreakRotator(InRot=BreakTransform(InTransform=@n16).Rotation).Yaw) -> (Output_Get:double)
   n17 = BreakTransform(InTransform=GetDeltaTransformFromRefPose(self=GetOwnerSkeletalMeshComponent(), BoneName=thigh_l, BaseName=pelvis))
   Set thigh_rot_l(execute=@n1.then_2, thigh_rot_l=MapRangeClamped(Value=BreakRotator(InRot=@n17.Rotation).Yaw, InRangeA=0, InRangeB=-110))
   n18 = GetDeltaTransformFromRefPose(self=GetOwnerSkeletalMeshComponent(), BoneName=lowerarm_r, BaseName=upperarm_r)
   Set elbow_rot_r(execute=@n1.then_1, elbow_rot_r:double=BreakRotator(InRot=BreakTransform(InTransform=@n18).Rotation).Yaw) -> (Output_Get:double)
   TryGetPawnOwner()
  ]}
]
variables [
 crouchingAlpha {type=number}
 elbow_rot_l {type=number}
 elbow_rot_r {type=number}
 isInFrontEndAlpha {type=number}
 isPistol {type=boolean}
 knee_rot_l {type=number}
 knee_rot_r {type=number}
 l_arm_dn {type=number}
 l_arm_fwd {type=number}
 l_arm_up {type=number}
 LoinClothRotator_01 {type=Rotator}
 LoinClothRotator_02 {type=Rotator}
 r_arm_dn {type=number}
 r_arm_fwd {type=number}
 r_arm_up {type=number}
 r_pistonAimUp_alpha {type=number}
 thigh_rot_l {type=number}
 thigh_rot_out_r {type=number}
 thigh_rot_r {type=number}
]
