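# Physics asset for the hair on the Quarrel medium male player cosmetic's FaceAcc mesh part, defining capsule bodies for dynamic hair motion.
# A head capsule anchors two-link front and side chains and three-link back chains, each link constrained to its predecessor.
# A two-link right tip chain branches from the first front-right bone.
# Narrow swing and twist limits keep strands near their shape, with the tip chain loosest.
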PhysicsAsset /BRCosmetics/Characters/Player/Male/Medium/Bodies/M_MED_Quarrel/Meshes/Parts/M_MED_Quarrel_FaceAcc_dyn_hair
bodies {
 dyn_hairTip_r_01 [{shapeType=Capsule, length=2.5, radius=1.5, shapeName=dyn_hair_ft_r_01_capsule, center=<1 0 0>, rotation=<rot 90 0 0>}]
 dyn_hairTip_r_02 [{shapeType=Capsule, length=2.5, radius=1.5, shapeName=dyn_hair_ft_r_02_capsule, center=<1 0 0>, rotation=<rot 90 0 0>}]
 dyn_hair_bk_01 [{shapeType=Capsule, length=2.25, radius=2, shapeName=dyn_hair_bk_01_capsule, center=<1.5 0 0>, rotation=<rot 90 0 0>}]
 dyn_hair_bk_02 [{shapeType=Capsule, length=2.25, radius=2, shapeName=dyn_hair_bk_02_capsule, center=<1.5 0 0>, rotation=<rot 90 0 0>}]
 dyn_hair_bk_03 [{shapeType=Capsule, length=2.25, radius=2, shapeName=dyn_hair_bk_03_capsule, center=<1.5 0 0>, rotation=<rot 90 0 0>}]
 dyn_hair_bk_l_01 [{shapeType=Capsule, length=2.25, radius=2, shapeName=dyn_hair_bk_l_01_capsule, center=<1 0 0>, rotation=<rot 90 0 0>}]
 dyn_hair_bk_l_02 [{shapeType=Capsule, length=2.25, radius=2, shapeName=dyn_hair_bk_l_02_capsule, center=<1 0 0>, rotation=<rot 90 0 0>}]
 dyn_hair_bk_l_03 [{shapeType=Capsule, length=2.25, radius=2, shapeName=dyn_hair_bk_l_03_capsule, center=<1 0 0>, rotation=<rot 90 0 0>}]
 dyn_hair_bk_r_01 [{shapeType=Capsule, length=2.25, radius=2, shapeName=dyn_hair_bk_r_01_capsule, center=<2 0 0>, rotation=<rot 90 0 0>}]
 dyn_hair_bk_r_02 [{shapeType=Capsule, length=2.25, radius=2, shapeName=dyn_hair_bk_r_02_capsule, center=<2 0 0>, rotation=<rot 90 0 0>}]
 dyn_hair_bk_r_03 [{shapeType=Capsule, length=2.25, radius=2, shapeName=dyn_hair_bk_r_03_capsule, center=<2 0 0>, rotation=<rot 90 0 0>}]
 dyn_hair_ft_01 [{shapeType=Capsule, length=4, radius=2, shapeName=dyn_hair_ft_01_capsule, center=<2 0 0>, rotation=<rot 90 0 0>}]
 dyn_hair_ft_02 [{shapeType=Capsule, length=4, radius=2, shapeName=dyn_hair_ft_02_capsule, center=<2 0 0>, rotation=<rot 90 0 0>}]
 dyn_hair_ft_l_01 [{shapeType=Capsule, length=4, radius=2, shapeName=dyn_hair_ft_l_01_capsule, center=<2 0 0>, rotation=<rot 90 0 0>}]
 dyn_hair_ft_l_02 [{shapeType=Capsule, length=4, radius=2, shapeName=dyn_hair_ft_l_02_capsule, center=<2 0 0>, rotation=<rot 90 0 0>}]
 dyn_hair_ft_r_01 [{shapeType=Capsule, length=4, radius=2, shapeName=dyn_hair_ft_r_01_capsule, center=<2 0 0>, rotation=<rot 90 0 0>}]
 dyn_hair_ft_r_02 [{shapeType=Capsule, length=4, radius=2, shapeName=dyn_hair_ft_r_02_capsule, center=<2 0 0>, rotation=<rot 90 0 0>}]
 dyn_hair_l_a_01 [{shapeType=Capsule, length=4, radius=2, shapeName=dyn_hair_l_a_01_capsule, center=<2 0 0>, rotation=<rot 90 0 0>}]
 dyn_hair_l_a_02 [{shapeType=Capsule, length=4, radius=2, shapeName=dyn_hair_l_a_02_capsule, center=<2 0 0>, rotation=<rot 90 0 0>}]
 dyn_hair_l_b_01 [{shapeType=Capsule, length=4, radius=2, shapeName=dyn_hair_l_b_01_capsule, center=<2 0 0>, rotation=<rot 90 0 0>}]
 dyn_hair_l_b_02 [{shapeType=Capsule, length=4, radius=2, shapeName=dyn_hair_l_b_02_capsule, center=<2 0 0>, rotation=<rot 90 0 0>}]
 dyn_hair_r_a_01 [{shapeType=Capsule, length=4, radius=2, shapeName=dyn_hair_r_a_01_capsule, center=<2 0 0>, rotation=<rot 90 0 0>}]
 dyn_hair_r_a_02 [{shapeType=Capsule, length=4, radius=2, shapeName=dyn_hair_r_a_02_capsule, center=<2 0 0>, rotation=<rot 90 0 0>}]
 dyn_hair_r_b_01 [{shapeType=Capsule, length=4, radius=2, shapeName=dyn_hair_r_b_01_capsule, center=<2 0 0>, rotation=<rot 90 0 0>}]
 dyn_hair_r_b_02 [{shapeType=Capsule, length=4, radius=2, shapeName=dyn_hair_r_b_02_capsule, center=<2 0 0>, rotation=<rot 90 0 0>}]
 head [{shapeType=Capsule, length=5, radius=5}]
}
constraints [
 {bone1Name=dyn_hair_bk_l_01, bone2Name=head, swing1LimitDegrees=6, swing2LimitDegrees=15, twistLimitDegrees=5}
 {bone1Name=dyn_hair_bk_l_02, bone2Name=dyn_hair_bk_l_01, swing1LimitDegrees=6, swing2LimitDegrees=15, twistLimitDegrees=5}
 {bone1Name=dyn_hair_bk_l_03, bone2Name=dyn_hair_bk_l_02, swing1LimitDegrees=6, swing2LimitDegrees=15, twistLimitDegrees=5}
 {bone1Name=dyn_hair_ft_01, bone2Name=head, swing1LimitDegrees=5, swing2LimitDegrees=15, twistLimitDegrees=2}
 {bone1Name=dyn_hair_ft_02, bone2Name=dyn_hair_ft_01, swing1LimitDegrees=5, swing2LimitDegrees=15, twistLimitDegrees=2}
 {bone1Name=dyn_hair_ft_l_01, bone2Name=head, swing1LimitDegrees=5, swing2LimitDegrees=15, twistLimitDegrees=2}
 {bone1Name=dyn_hair_ft_l_02, bone2Name=dyn_hair_ft_l_01, swing1LimitDegrees=5, swing2LimitDegrees=15, twistLimitDegrees=2}
 {bone1Name=dyn_hair_l_a_01, bone2Name=head, swing1LimitDegrees=5, swing2LimitDegrees=15, twistLimitDegrees=2}
 {bone1Name=dyn_hair_l_a_02, bone2Name=dyn_hair_l_a_01, swing1LimitDegrees=5, swing2LimitDegrees=15, twistLimitDegrees=2}
 {bone1Name=dyn_hair_l_b_01, bone2Name=head, swing1LimitDegrees=5, swing2LimitDegrees=15, twistLimitDegrees=2}
 {bone1Name=dyn_hair_l_b_02, bone2Name=dyn_hair_l_b_01, swing1LimitDegrees=5, swing2LimitDegrees=15, twistLimitDegrees=2}
 {bone1Name=dyn_hair_ft_r_01, bone2Name=head, swing1LimitDegrees=5, swing2LimitDegrees=15, twistLimitDegrees=2}
 {bone1Name=dyn_hair_ft_r_02, bone2Name=dyn_hair_ft_r_01, swing1LimitDegrees=5, swing2LimitDegrees=15, twistLimitDegrees=2}
 {bone1Name=dyn_hair_r_a_01, bone2Name=head, swing1LimitDegrees=5, swing2LimitDegrees=15, twistLimitDegrees=2}
 {bone1Name=dyn_hair_r_a_02, bone2Name=dyn_hair_r_a_01, swing1LimitDegrees=5, swing2LimitDegrees=15, twistLimitDegrees=2}
 {bone1Name=dyn_hair_r_b_01, bone2Name=head, swing1LimitDegrees=5, swing2LimitDegrees=15, twistLimitDegrees=2}
 {bone1Name=dyn_hair_r_b_02, bone2Name=dyn_hair_r_b_01, swing1LimitDegrees=5, swing2LimitDegrees=15, twistLimitDegrees=2}
 {bone1Name=dyn_hair_bk_r_01, bone2Name=head, swing1LimitDegrees=6, swing2LimitDegrees=15, twistLimitDegrees=5}
 {bone1Name=dyn_hair_bk_r_02, bone2Name=dyn_hair_bk_r_01, swing1LimitDegrees=6, swing2LimitDegrees=15, twistLimitDegrees=5}
 {bone1Name=dyn_hair_bk_r_03, bone2Name=dyn_hair_bk_r_02, swing1LimitDegrees=6, swing2LimitDegrees=15, twistLimitDegrees=5}
 {bone1Name=dyn_hair_bk_01, bone2Name=head, swing1LimitDegrees=6, swing2LimitDegrees=15, twistLimitDegrees=5}
 {bone1Name=dyn_hair_bk_02, bone2Name=dyn_hair_bk_01, swing1LimitDegrees=6, swing2LimitDegrees=15, twistLimitDegrees=5}
 {bone1Name=dyn_hair_bk_03, bone2Name=dyn_hair_bk_02, swing1LimitDegrees=6, swing2LimitDegrees=15, twistLimitDegrees=5}
 {bone1Name=dyn_hairTip_r_01, bone2Name=dyn_hair_ft_r_01, swing1LimitDegrees=10, swing2LimitDegrees=20, twistLimitDegrees=5}
 {bone1Name=dyn_hairTip_r_02, bone2Name=dyn_hairTip_r_01, swing1LimitDegrees=10, swing2LimitDegrees=20, twistLimitDegrees=5}
]
