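# Animation Blueprint for the male Quarrel character's face accessory mesh, driving its hair bones on top of a pose copied from the attached parent.
# By default, a RigidBody node simulates the hair in head-bone space with overridden gravity.
# While the Quarrel emote montage plays on the body's anim instance, a bool blend swaps physics for an authored hair sequence layered onto the hair bone branches.
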AnimBlueprint /BRCosmetics/Characters/Player/Male/Medium/Bodies/M_MED_Quarrel/Meshes/Parts/M_MED_Quarrel_FaceAcc_AnimBP
target_skeleton=&/M_MED_Quarrel_FaceAcc_Skeleton
graphs [
 AnimGraph {nodes [
   Copy Pose From Mesh() {bCopyCurves=true, bUseAttachedParent=true}
   Layered blend per bone(BlendPoses_0=Copy Pose From Mesh(SourceMeshComponent=Get Body_Mesh()) {bCopyCurves=true, bUseAttachedParent=true}) {layerSetup [{branchFilters [{boneName=thigh_l}, {boneName=thigh_r}, {boneName=calf_l}, {boneName=calf_r}]}]}
   n1 = RigidBody(ComponentPose=Local To Component(LocalPose=Use cached pose 'pre-emote'())) {
    bClampLinearTranslationLimitToRefPose=true
    bOverrideWorldGravity=true
    overridePhysicsAsset=&/M_MED_Quarrel_FaceAcc_dyn_hair
    simulationSpace=BaseBoneSpace
    alphaBoolBlend {blendOption=Linear}
    baseBoneRef {boneName=head}
    componentAppliedLinearAccClamp=<100 200 200>
    overrideWorldGravity {z=480}
    simSpaceSettings {maxAngularAcceleration=20, maxAngularVelocity=2, maxLinearAcceleration=30, maxLinearVelocity=5, velocityScaleZ=0, worldAlpha=1}
   }
   n2 = Layered blend per bone(
    BasePose=Use cached pose 'pre-emote'()
    BlendPoses_0=Sequence Player 'Emote_QuarrelMale_Hair'!(StartPosition:float=0) -> (Pose) {bLoopAnimation=false, sequence=/BRCosmetics/Animation/Game/MainPlayer/Emotes/QuarrelMale/Emote_QuarrelMale_Hair}
   ) {layerSetup [{branchFilters [
       {boneName=dyn_hair_ft_01}
       {boneName=dyn_hair_ft_l_01}
       {boneName=dyn_hair_l_a_01}
       {boneName=dyn_hair_l_b_01}
       {boneName=dyn_hair_bk_l_02}
       {boneName=dyn_hair_bk_02}
       {boneName=dyn_hair_bk_r_02}
       {boneName=dyn_hair_r_b_01}
       {boneName=dyn_hair_r_a_01}
       {boneName=dyn_hair_ft_r_01}
      ]}]}
   Output Pose(Result=Blend Poses by bool(
     BlendPose_0=@n2
     BlendPose_1=Component To Local(ComponentPose=@n1)
     bActiveValue=Get isCustomEmotePlaying_hair() -> (isCustomEmotePlaying_hair:bool)
    ) {blendType=Linear, childUpateMode=ResetChildOnActivate}) {layerGroup=None}
   Save cached pose 'pre-emote'(Pose=Copy Pose From Mesh() {bCopyCurves=true, bUseAttachedParent=true}) {cacheName=pre-emote}
  ]}
 EventGraph {nodes [
   Event BlueprintUpdateAnimation()
   > Set Body_Mesh(Body_Mesh=GetOwnerSkeletalMeshComponent())
   > n1 = Cast To PlayerPawn_Athena(Object=TryGetPawnOwner())
   > n2 = Set Pawn Owner(Pawn Owner:Player Pawn Athena Object Reference=@n1) -> (Output_Get:Player Pawn Athena Object Reference)
   > Set Head_Mesh(Head_Mesh:Skeletal Mesh Component Object Reference=GetSkeletalMeshForPartType(self=@n2, PartType=Head)) -> (Output_Get:Skeletal Mesh Component Object Reference)
   > Set isCustomEmotePlaying_hair(isCustomEmotePlaying_hair:bool=false) -> (Output_Get:bool)
   n3 = GetAnimInstance(self=GetOwnerSkeletalMeshComponent())
   > Is Valid(InputObject=@n3)
   > Set isCustomEmotePlaying_hair(isCustomEmotePlaying_hair=Montage_IsPlaying(self=@n3, Montage=/BRCosmetics/Animation/Game/MainPlayer/Emotes/QuarrelMale/Emote_QuarrelMale_CMM_M))
   Set isCustomEmotePlaying(isCustomEmotePlaying:double=SelectFloat(A=0, B=1, bPickA=OR Boolean!(A=false, B=false, C:bool, D:bool, E:bool) -> (ReturnValue))) -> (Output_Get:double)
  ]}
]
variables [
 Body_Mesh {type=SkeletalMeshComponent}
 Head_Mesh {type=SkeletalMeshComponent}
 isCustomEmotePlaying {type=number}
 isCustomEmotePlaying_hair {type=boolean}
 Pawn Owner {type=PlayerPawn_Athena_C}
]
